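# Translucent UI material for the Dual Revolvers reticle that draws a green, curved progress meter.
# A circle outline is clipped to a rotatable arc between start and end angles.
# Radial arc functions turn MeterProgress and DeltaProgress into anti-aliased fills, combined with a faint background for opacity.
# Time-scrolled horizontal scanlines tint the colors, while an inset outline mask adds shading inside the band.
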Material /CherrySmokeDualRevolversGameplay/UI/M_UI_DualRevolvers_Reticle
graph {
 properties {bAutomaticallySetUsageInEditor=false, blendMode=BLEND_Translucent, materialDomain=MD_UI, usedShadingModels=Default Lit}
 nodes [
  n1 = Time()
  n2 = ScalarParameter() {defaultValue=2.5, parameterName=Speed}
  n3 = ComponentMask(None=OneMinus(None=TextureCoordinate())) {g=true}
  n4 = ScalarParameter() {defaultValue=100, parameterName=Number of Lines}
  n5 = Frac(None=Add(A=Multiply(A=@n1, B=@n2), B=Multiply(A=@n3, B=@n4)))
  n6 = SmoothStep(
   Min=ScalarParameter() {defaultValue=-0.0691, parameterName=Line Thickness}
   Max=ScalarParameter() {defaultValue=0.4, parameterName=Line Blur}
   Value=Abs(None=Subtract(A=@n5) {constB=0.5})
  )
  n7 = Multiply(A=@n6) {constB=0.25}
  n8 = MaterialFunctionCall(
   Size=ScalarParameter() {defaultValue=0.4, group=Circle Parameters, parameterName=CircleSize, sliderMax=1, sortPriority=4}
   OutlineThickness=ScalarParameter() {defaultValue=0.025, group=Circle Parameters, parameterName=CircleThickness, sliderMax=1, sortPriority=4}
  ) {materialFunction=&MF_UI_Circle}
  n9 = MaterialFunctionCall(
   UVs=MaterialFunctionCall() {materialFunction=&GetUserInterfaceUV}.Normalized UV
   "Rotation Angle (0-1)"=ScalarParameter() {group=Radial Parameters, parameterName=MeterRotation, sliderMax=1, sortPriority=3}
  ) {materialFunction=&CustomRotator}
  n10 = ScalarParameter() {defaultValue=0.25, group=Circle Parameters, parameterName=StartAngle, sliderMax=1, sortPriority=4}
  n11 = ScalarParameter() {defaultValue=0.505, group=Circle Parameters, parameterName=EndAngle, sliderMax=1, sortPriority=4}
  n12 = ScalarParameter() {group=Circle Parameters, parameterName=FillDirection, sliderMax=1, sliderMin=-1, sortPriority=4}
  n13 = MaterialFunctionCall!(
   UV=@n9.Rotated Values
   "Arc Start Angle (0-1)"=@n10
   "Arc End Angle (0-1)"=@n11
   "Fill Ratio (0-1)"=OneMinus(None=ScalarParameter() {defaultValue=0.5, group=Progress Parameters, parameterName=MeterProgress, sliderMax=1, sortPriority=1})
   Fill Clockwise=@n12
  ) {materialFunction=/FortUILibrary/MaterialFunctions/Basics/MF_UI_RadialArc}
  n14 = Ceil(None=@n13.Arc Mask)
  n15 = Multiply(A=@n8.Outline, B=@n14)
  n16 = Multiply(
   A=MaterialFunctionCall!(DistanceField=@n8.SDF, Threshold, Thickness=Constant() {r=0.0125}, Alignment, Smoothing) {materialFunction=/FortUILibrary/MaterialFunctions/Shapes/Modifiers/MF_UI_ShapeOutline}.Outline
   B=@n14
  )
  n17 = Constant()
  n18 = Constant()
  n19 = Multiply(
   A=Min(A=@n15, B=OneMinus(None=@n16))
   B=Multiply(A=MaterialFunctionCall(UV=@n9.Rotated Values, "Arc Start Angle (0-1)"=@n10, "Arc End Angle (0-1)"=@n11, "Fill Ratio (0-1)"=@n17, Fill Clockwise=@n18) {materialFunction=&MF_UI_RadialArc}.Centered Fill) {constB=2}
  )
  n20 = NamedRerouteUsage()
  LinearInterpolate(
   A=Multiply(A=OneMinus(None=@n7), B=Constant4Vector() {constant=<rgba 0.1499 0.3826 0.1942 1>}.RGBA) {constB=0.1}
   B=NamedRerouteUsage()
   Alpha=Add(A=Multiply(A=@n19, B=@n20), B=@n20)
  )
  n21 = Constant() {r=0.001}
  n22 = MaterialFunctionCall(DistanceField=@n13.Radial Fill, Threshold=@n21) {materialFunction=&MF_UI_AntiAliasing}
  n23 = MaterialFunctionCall(
   UV=@n9.Rotated Values
   "Arc Start Angle (0-1)"=@n10
   "Arc End Angle (0-1)"=@n11
   "Fill Ratio (0-1)"=OneMinus(None=ScalarParameter() {group=Progress Parameters, parameterName=DeltaProgress, sliderMax=1, sortPriority=1})
   Fill Clockwise=@n12
  ) {materialFunction=&MF_UI_RadialArc}
  n24 = MaterialFunctionCall(DistanceField=@n23.Radial Fill, Threshold=@n21) {materialFunction=&MF_UI_AntiAliasing}
  n25 = Ceil(None=OneMinus(None=@n23.Arc Mask))
  n26 = Max(
   A=Max(
    A=NamedRerouteDeclaration(None=@n22.Outer) {name=radialMeterFill, nodeColor=<rgba 0.025 0.0203 0.0125 0.75>}
    B=NamedRerouteDeclaration(None=@n24.Outer) {name=radialDeltaFill, nodeColor=<rgba 0.025 0.0203 0.0125 0.75>}
   )
   B=Multiply(A=Saturate(None=@n25), B=ScalarParameter() {defaultValue=0.5, parameterName=BackgroundOpacity, sliderMax=1, sortPriority=4}) {constB=0.5}
  )
  Multiply(
   A=Add(A=Multiply(A=@n15, B=@n26), B=Multiply(A=@n19) {constB=0.1})
   B=ScalarParameter() {defaultValue=0.9, parameterName=MasterOpacity, sliderMax=1}
  )
  n27 = Constant4Vector() {constant=<rgba 0 0.5781 0 0>}
  n28 = Constant4Vector() {constant=<rgba 0.2341 0.7135 0.2398 1>}
  NamedRerouteDeclaration(None=LinearInterpolate(A=LinearInterpolate(A=@n27.RGBA, B=@n28.RGBA, Alpha=@n6), B=Constant4Vector() {constant=<rgba 0.4504 0.9826 0.5079 1>}.RGBA) {constAlpha=0.3}) {name=ProgressColor, nodeColor=<rgba 0.025 0.0203 0.0125 0.75>}
 ]
}
