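# Reticle extension widget for the Cherry Smoke Dual Revolvers, showing a hangtime meter and tracking weapon spread.
# Construct binds to HUD spread changes, which resize a SizeBox, and to the weapon's Hangtime Active dispatcher.
# On activation, the bar appears and a looping timer feeds a countdown timer's remaining fraction, normalized by curve-table hangtime, into the material's MeterProgress.
# Deactivation hides the bar and clears both timers; countdown expiry only clears timers.
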WidgetBlueprint /CherrySmokeDualRevolversGameplay/UI/WBP_CherrySmokeDualRevolvers_Reticle
graphs [
 EventGraph {nodes [
   n1 = SetHangtimeMeterTimerProgress() {is=title K2Node_CustomEvent}
   n2 = Get HangtimeTotalTimeTimerHandle() -> (HangtimeTotalTimeTimerHandle:Timer Handle Structure)
   n3 = GetDynamicMaterial(execute=@n6.then, self=Get Image_HangtimeBar() -> (Image_HangtimeBar:Image Object Reference))
   SetScalarParameterValue(
    execute=@n1.then
    self=@n3
    ParameterName=MeterProgress
    Value=SafeDivide(
     A=Get Timer Remaining Time by Handle!(Handle:Timer Handle Structure=@n2) -> (ReturnValue:float) {is=title K2Node_CallFunction}
     B=Get HangtimeTimeFloat() -> (HangtimeTimeFloat:float)
    )
   )
   n4 = HangtimeTimerEndedEvent() {is=title K2Node_CustomEvent}
   > Clear and Invalidate Timer by Handle(Handle=Get HangtimeTotalTimeTimerHandle())
   > Clear and Invalidate Timer by Handle(Handle=Get HangtimeUpdateTimerHandle() -> (HangtimeUpdateTimerHandle:Timer Handle Structure))
   n5 = HangtimeActiveUIEvent!() -> (OutputDelegate, then, HangtimeActive:bool=~, TimeRemaining:double=~) {is=title K2Node_CustomEvent}
   > n6 = Branch(Condition=@n5.HangtimeActive)
   SetOpacity!(execute=@n6.else, self:Image Object Reference=Get Image_HangtimeBar(), InOpacity:float=0) -> (then)
   > Clear and Invalidate Timer by Handle(Handle=Get HangtimeUpdateTimerHandle())
   > Clear and Invalidate Timer by Handle(Handle=Get HangtimeTotalTimeTimerHandle())
   Event Construct()
   > Parent: Construct()
   > Bind Event to On Spread Radius Changed(
    self:Fort HUDContext Object Reference=Get FortHUDContext!() -> (ReturnValue:Fort HUDContext Object Reference) {is=title K2Node_GetBlueprintContextEx}
    Delegate=Create Event()
   ) {is=title K2Node_AddDelegate}
   > n7 = Cast To FortWeaponRanged!(
    execute
    Object=GetEquippedWeapon!(self:Fort Weapon Reticle Extension Widget Base Object Reference) -> (ReturnValue:Fort Weapon Object Reference)
   ) -> (then, CastFailed, AsFort Weapon Ranged:Fort Weapon Ranged Object Reference) {is=title K2Node_DynamicCast}
   > Set EquippedWeapon(EquippedWeapon:Fort Weapon Ranged Object Reference=@n7) -> (Output_Get:Fort Weapon Ranged Object Reference)
   > SetOpacity(self=Get Image_HangtimeBar(), InOpacity=0)
   > Set HangtimeTimeFloat(HangtimeTimeFloat:float=GetValueAtLevel!(Input:"Scalable Float Structure (by ref)"=Get HF_HangtimeTime() -> (HF_HangtimeTime:Scalable Float Structure), Level:float=0) -> (ReturnValue:float)) -> (Output_Get:float)
   > n8 = Cast To B_CherrySmoke_DualRevolvers_Athena!(execute, Object=Get EquippedWeapon() -> (EquippedWeapon:Fort Weapon Ranged Object Reference)) -> (then, CastFailed, AsB Cherry Smoke Dual Revolvers Athena:B Cherry Smoke Dual Revolvers Athena Object Reference) {is=title K2Node_DynamicCast}
   > Bind Event to Hangtime Active Dispatcher(self:B Cherry Smoke Dual Revolvers Athena Object Reference=@n8, Delegate=Create Event()) {is=title K2Node_AddDelegate}
   Event Destruct()
   > Parent: Destruct()
   > Unbind Event from On Spread Radius Changed(self:Fort HUDContext Object Reference=Get FortHUDContext(), Delegate=Create Event()) {is=title K2Node_RemoveDelegate}
   n9 = Get EquippedWeapon()
   > Is Valid(InputObject=@n9)
   > n10 = Cast To B_CherrySmoke_DualRevolvers_Athena(Object=@n9)
   > Unbind Event from Hangtime Active Dispatcher(self:B Cherry Smoke Dual Revolvers Athena Object Reference=@n10, Delegate=Create Event()) {is=title K2Node_RemoveDelegate}
   Get HangtimeTimeFloat()
   n11 = Set Timer by Event(execute=@n13.then, Delegate=@n1, Time=0.05, bLooping=true, bMaxOncePerFrame=true)
   > Set HangtimeUpdateTimerHandle(HangtimeUpdateTimerHandle:Timer Handle Structure=@n11) -> (Output_Get:Timer Handle Structure)
   n12 = Set Timer by Event(execute=@n15.then, Delegate=@n4, Time=@n5.TimeRemaining, bLooping=false, bMaxOncePerFrame=true)
   > n13 = Set HangtimeTotalTimeTimerHandle(HangtimeTotalTimeTimerHandle:Timer Handle Structure=@n12) -> (Output_Get:Timer Handle Structure)
   n14 = SafeDivide(A=Get Timer Remaining Time by Handle(Handle=Get HangtimeTotalTimeTimerHandle()), B=@n5.TimeRemaining)
   n15 = SetOpacity(execute=@n16.then, self=Get Image_HangtimeBar(), InOpacity=1)
   n16 = SetScalarParameterValue(execute=@n3.then, self=@n3, ParameterName=MeterProgress, Value=@n14)
  ]}
 OnSpreadRadiusChanged {nodes [
   n1 = OnSpreadRadiusChanged!() -> (then, SpreadRadius:float=~) {is=title K2Node_FunctionEntry}
   n2 = Get SizeBox_Spread() -> (SizeBox_Spread:Size Box Object Reference)
   > SetWidthOverride(self=@n2, InWidthOverride=@n1)
   > SetHeightOverride(self=@n2, InHeightOverride=@n1)
  ]}
]
variables [
 DelayUntilHideOverheat {type=number, default=2}
 EquippedWeapon {type=FortWeaponRanged, default=null}
 HangtimeTimeFloat {type=number}
 HangtimeTotalTimeTimerHandle {type=TimerHandle}
 HangtimeUpdateTimerHandle {type=TimerHandle}
 HF_HangtimeTime {
  type=ScalableFloat
  default {
   registryType=None
   value=1
   curve {
    curveTable=/CherrySmokeDualRevolversGameplay/DataTables/CherrySmokeDualRevolversGameplayGameData
    rowName=CherrySmokeDualRevolvers.Hangtime.TimeBeforeEnd
   }
  }
 }
 Time Remaining {type=number}
]
widgets {
 info {
  inheritedWidgetCount=1
  parentClass=/Script/FortniteUI.FortWeaponReticleExtensionWidgetBase
  rootWidgetClass=&Overlay
  widgetCount=4
 }
 widgets [
  OverlayAll {widgetClassPath=&Overlay, widget=OverlayAll}
  Image_HangtimeBar {
   widgetClassPath=&Image
   bIsVariable=true
   parent=OverlayAll
   widget=Image_HangtimeBar
   properties {
    brush {
     bIsDynamicallyLoaded=false
     drawAs=Image
     imageType=FullColor
     mirroring=NoMirror
     resourceName=None
     resourceObject=&/M_UI_DualRevolvers_Reticle
     tiling=NoTile
     imageSize=<32 32>
     margin=<ltrb 0 0 0 0>
     outlineSettings {
      bUseBrushTransparency=false
      roundingType=HalfHeightRadius
      width=0
      color {colorUseRule=UseColor_Specified, specifiedColor=<rgba 0 0 0 0>}
      cornerRadii=<0 0 0 0>
     }
     tintColor {colorUseRule=UseColor_Specified, specifiedColor=<rgba 1 1 1 1>}
     uVRegion {bIsValid=false, max=<0 0>, min=<0 0>}
    }
    renderTransform {angle=0, scale=<10 10>, shear=<0 0>, translation=<0 0>}
   }
   slot {class=OverlaySlot, properties {horizontalAlignment=HAlign_Center, verticalAlignment=VAlign_Center, padding=<ltrb 43 43 0 0>}}
  }
  SizeBox_Spread {
   widgetClassPath=&SizeBox
   bIsVariable=true
   parent=OverlayAll
   widget=SizeBox_Spread
   properties {bOverride_HeightOverride=true, bOverride_WidthOverride=true}
   slot {class=OverlaySlot, properties {horizontalAlignment=HAlign_Center, verticalAlignment=VAlign_Center}}
  }
  TutorialHighlight {widgetClassPath=&TutorialHighlight}
 ]
}
